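# Material function for the HealedScar Sand player skin, layering ghostly glowing wisps and scar emissive over incoming material attributes.
# Pre-skinned position and normal drive triplanar-style blending of three panning smoke-noise layers, distorted by nested flowmaps and given depth by BumpOffset.
# A mask texture separates cloth, armor and scars for HSV base-color shifts and glow intensity.
# Emissive combines Fresnel silhouette glow, tinted wisps and scars, with roughness and specular also adjusted.
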MaterialFunction /BRCosmetics/Characters/Player/Male/Large/Bodies/M_LRG_HealedScar/Skins/Sand/Materials/MF_HealedScar_Sand
graph {nodes [
  Constant() {r=0.05}
  n1 = GetMaterialAttributes(None=FunctionInput() {inputType=FunctionInput_MaterialAttributes}) {attributeGetTypes [
    #1
    #2
    #3
    #4
    #5
   ]}
  n2 = Multiply(A=ComponentMask(None=NamedRerouteUsage()) {r=true}) {constB=4}
  n3 = SetMaterialAttributes!(
   MaterialAttributes=@n1.MaterialAttributes
   EmissiveColor=NamedRerouteUsage()
   Normal=@n1.MaterialAttributes
   BaseColor=NamedRerouteUsage()
   Roughness=Add(A=@n1.Roughness, B=@n2)
   Specular=LinearInterpolate(A=@n1.Specular, Alpha=@n2) {constB=0}
  ) {attributeSetTypes [
    #1
    #2
    #3
    #4
    #5
   ]}
  FunctionOutput(None=@n3) {outputName=OUT}
  NamedRerouteDeclaration(None=ScalarParameter() {defaultValue=1, group=HealedScar_Sand, parameterName=INAlpha}) {name=Alpha, nodeColor=<rgba 1 0.7969 0 1>}
  n4 = NamedRerouteDeclaration(None=VertexInterpolator(VS=LocalPosition() {localOrigin=InstancePreSkinning}.XYZ).PS) {name=PreskinnedPos, nodeColor=<rgba 1 0.2109 0 1>}
  n5 = Power(Base=SmoothStep(Value=ComponentMask(None=@n4) {g=true}) {constMax=15, constMin=8, desc=Belly})
  n6 = NamedRerouteUsage()
  NamedRerouteDeclaration(None=Multiply(A=Multiply(A=@n5, B=@n6)) {constB=0.5}) {name=BellyBrighness, nodeColor=<rgba 0 0.5312 1 1>}
  NamedRerouteDeclaration(None=@n1.EmissiveColor) {name=OriginalEmissiveColor, nodeColor=<rgba 1 0 0.9141 1>}
  NamedRerouteDeclaration(None=@n1.BaseColor) {name=OriginalBaseColor, nodeColor=<rgba 0 1 0.1797 1>}
  n7 = MaterialFunctionCall(Float3=RgbToHsv!(None=NamedRerouteUsage())) {materialFunction=&BreakOutFloat3Components}
  n8 = LinearInterpolate(
   A=Add(A=@n7.R, B=ScalarParameter() {defaultValue=0.445, group=HealedScar_Sand, parameterName=HUE_Offset_NoneArmor})
   B=Add(A=@n7.R, B=ScalarParameter() {group=HealedScar_Sand, parameterName=HUE_Offset_Armor})
   Alpha=NamedRerouteUsage()
  )
  n9 = LinearInterpolate(
   A=Add(A=@n7.G, B=ScalarParameter() {defaultValue=0.064, group=HealedScar_Sand, parameterName=Saturation_Offset_NoneArmor})
   B=Add(A=@n7.G, B=ScalarParameter() {group=HealedScar_Sand, parameterName=Saturation_Offset_Armor})
   Alpha=NamedRerouteUsage()
  )
  n10 = ScalarParameter() {group=HealedScar_Sand, parameterName=Vaule_Offset_NoneArmor}
  n11 = ScalarParameter() {group=HealedScar_Sand, parameterName=Vaule_Offset_Armor}
  n12 = LinearInterpolate(A=Add(A=@n7.B, B=Divide(A=@n10) {constB=10}), B=Add(A=@n7.B, B=Divide(A=@n11) {constB=10}), Alpha=NamedRerouteUsage())
  n13 = LinearInterpolate(
   A=HsvToRgb!(None=MaterialFunctionCall(X=@n8, Y=@n9, Z=@n12) {materialFunction=&MakeFloat3}.Result)
   B=NamedRerouteUsage()
   Alpha=NamedRerouteUsage()
  )
  n14 = Multiply(A=NamedRerouteUsage(), B=OneMinus(None=NamedRerouteUsage()))
  NamedRerouteDeclaration(None=StaticSwitchParameter(True=Add(A=@n13, B=@n14), False=@n13) {defaultValue=true, group=HealedScar_Sand, parameterName=IsBodyPart}) {name=BaseColor, nodeColor=<rgba 0.3125 1 0 1>}
  n15 = VertexInterpolator(VS=Add(A=LocalPosition().XYZ, B=Constant3Vector() {constant=<rgba 0 0 -125 0>}))
  n16 = Constant()
  NamedRerouteDeclaration(None=SmoothStep(Value=SphereMask(A=@n15.PS, B=@n16) {attenuationRadius=30, hardnessPercent=1}) {constMax=0.5}) {name=ChestMask, nodeColor=<rgba 0.5469 0 1 1>}
  n17 = TextureSampleParameter2D() {
   group=HealedScar_Sand
   parameterName=HealScar_TechArt_Mask
   texture=/BRCosmetics/Characters/Player/Male/Large/Bodies/M_LRG_HealedScar/Skins/Sand/Textures/T_HealedScar_Sand_Body_TechArt_Mask
  }
  NamedRerouteDeclaration(None=@n17.B) {name=ScarMask, nodeColor=<rgba 0 1 0 1>}
  NamedRerouteDeclaration(None=@n17.G) {name=MetalArmorMask, nodeColor=<rgba 0 0 0.991 1>}
  n18 = FunctionInput() {inputName=Fresnel, inputType=FunctionInput_Scalar, sortPriority=10}
  n19 = OneMinus(None=NamedRerouteUsage())
  n20 = Power(
   Base=Multiply(A=@n18, B=Saturate(None=@n19))
   Exp=ScalarParameter() {defaultValue=2, group=HealedScar_Sand, parameterName=GhostFresnelPower, sortPriority=10}
  )
  n21 = Multiply(
   A=@n20
   B=LinearInterpolate(
    A=ScalarParameter() {defaultValue=-0.5, group=HealedScar_Sand, parameterName=Fresnel_GlowIRange_Armor, sortPriority=10}
    B=ScalarParameter() {defaultValue=-0.75, group=HealedScar_Sand, parameterName=Fresnel_GlowIRange_NonArmor, sortPriority=10}
    Alpha=NamedRerouteUsage()
   )
  ) {constB=1.3}
  n22 = LinearInterpolate(A=Add(A=@n20, B=@n21), B=Multiply(A=Multiply(A=@n21) {constB=-1}) {constB=0.1}, Alpha=NamedRerouteUsage()) {constB=0.2}
  n23 = Multiply(A=Power(Base=InverseLinearInterpolate(Value=@n18) {constA=0.05, constB=0}) {constExponent=4}, B=NamedRerouteUsage())
  n24 = VectorParameter() {group=HealedScar_Sand, parameterName=SillhouetteGlowColor, sortPriority=10, defaultValue=<rgba 0.1858 0.9812 0.9911 1>}
  n25 = VectorParameter() {group=HealedScar_Sand, parameterName=WispsEmissiveColor_1, sortPriority=10, defaultValue=<rgba 0.031 0.522 0.75 1>}
  n26 = NamedRerouteUsage()
  n27 = VectorParameter() {group=HealedScar_Sand, parameterName=WispsEmissiveColor_2, sortPriority=10, defaultValue=<rgba 0.1248 0.4423 1 1>}
  n28 = NamedRerouteUsage()
  n29 = VectorParameter() {group=HealedScar_Sand, parameterName=WispsEmissiveColor_3, sortPriority=10, defaultValue=<rgba 0.4376 0.5293 0.5573 1>}
  n30 = Add(A=Add(A=Multiply(A=@n25.RGB, B=@n26), B=Multiply(A=@n27.RGB, B=@n28)), B=Multiply(A=@n29.RGB, B=NamedRerouteUsage()))
  n31 = NamedRerouteUsage()
  n32 = NamedRerouteUsage()
  n33 = Add(A=Add(A=Multiply(A=@n25.RGB, B=@n31), B=Multiply(A=@n27.RGB, B=@n32)), B=Multiply(A=@n29.RGB, B=NamedRerouteUsage()))
  n34 = NamedRerouteUsage()
  n35 = ScalarParameter() {defaultValue=0.2, group=HealedScar_Sand, parameterName=Wisps_Intensity_NoneArmor}
  n36 = ScalarParameter() {defaultValue=0.075, group=HealedScar_Sand, parameterName=Wisps_Intensity_Armor}
  n37 = NamedRerouteUsage()
  n38 = Add(
   A=Multiply(A=Add(A=@n22, B=@n23), B=@n24.RGB)
   B=Multiply(A=LinearInterpolate(A=@n30, B=@n33, Alpha=@n34) {desc=Float3}, B=LinearInterpolate(A=@n35, B=@n36, Alpha=@n37))
  )
  n39 = NamedRerouteUsage()
  n40 = NamedRerouteUsage()
  n41 = ScalarParameter() {defaultValue=1, group=HealedScar_Sand, parameterName=ScarIntensity}
  n42 = NamedRerouteUsage()
  NamedRerouteDeclaration(None=LinearInterpolate(A=LinearInterpolate(A=@n38, B=@n39, Alpha=@n40), B=Multiply(A=@n41, B=@n42), Alpha=NamedRerouteUsage())) {name=Emissive, nodeColor=<rgba 0 0.9646 1 1>}
  n43 = VectorParameter() {
   group=HealedScar_Sand
   parameterName=Wisps_Main_DistortUV
   channelNames {a=Offset.Y, b=Offset.X, g=Tilling.Y, r=Tilling.X}
   defaultValue=<rgba 1 1 0 0>
  }
  n44 = Add(
   A=Multiply(
    A=NamedRerouteUsage()
    B=MaterialFunctionCall(X=@n43.Tilling.X, Y=@n43.Tilling.X) {materialFunction=&MakeFloat2}.Result
   )
   B=MaterialFunctionCall(X=@n43.Offset.X, Y=@n43.Offset.X) {materialFunction=&MakeFloat2}.Result
  )
  n45 = VectorParameter() {group=HealedScar_Sand, parameterName=Wisps_Main_Flowmap_OffsetSpeed, defaultValue=<rgba -0.02 -0.02 0 0>}
  n46 = MaterialFunctionCall(X=@n45.R, Y=@n45.R) {materialFunction=&MakeFloat2}
  n47 = NamedRerouteUsage()
  n48 = ScalarParameter() {group=HealedScar_Sand, parameterName=FlowMap_Secondary_Scale}
  n49 = TextureSampleParameter2D(UVs=Add(A=Panner(Coordinate=@n44, Speed=@n46.Result), B=Multiply(A=@n47, B=@n48))) {
   desc=Only RG channel works
   group=HealedScar_Sand
   parameterName=Wisps_DistortUV_Flowmap
   samplerSource=SSM_Wrap_WorldGroupSettings
   texture=/BRCosmetics/Characters/Player/Male/Large/Bodies/M_LRG_HealedScar/Skins/Sand/Textures/T_HealedScar_Sand_Flowmap
  }
  n50 = Multiply(A=ComponentMask(None=@n49.RGB) {g=true, r=true}) {constB=2}
  NamedRerouteDeclaration(None=Multiply(A=Subtract(A=@n50), B=ScalarParameter() {group=HealedScar_Sand, parameterName=FlowMap_Main_Scale})) {name=FlowMap_Side_Main, nodeColor=<rgba 0 1 0.1797 1>}
  NamedRerouteDeclaration(None=@n17.R) {name=ClothMask, nodeColor=<rgba 0 0 0.991 1>}
  n51 = NamedRerouteUsage()
  n52 = VectorParameter() {
   group=HealedScar_Sand
   parameterName=WispsNoise1_UV
   channelNames {a=Offset Y, b=Offset X, g=Tilling Y, r=Tilling X}
   defaultValue=<rgba 0.75 0.75 0 0>
  }
  n53 = Multiply(
   A=@n51
   B=MaterialFunctionCall(X=@n52.Tilling X, Y=@n52.Tilling X) {materialFunction=&MakeFloat2}.Result
  )
  n54 = ComponentMask(None=MaterialFunctionCall(X=@n52.Offset X, Y=@n52.Offset X) {materialFunction=&MakeFloat2}.Result) {g=true, r=true}
  n55 = VectorParameter() {group=HealedScar_Sand, parameterName=WispsNoise1_OffsetSpeed, defaultValue=<rgba 0.01 0.02 0 0>}
  n56 = Add(
   A=Panner(
    Coordinate=Add(A=@n53, B=@n54)
    Speed=MaterialFunctionCall(X=@n55.R, Y=@n55.R) {materialFunction=&MakeFloat2}.Result
   )
   B=NamedRerouteUsage()
  )
  n57 = ScalarParameter() {group=HealedScar_Sand, parameterName=Wisps1_Height}
  n58 = NamedRerouteUsage()
  n59 = TextureObjectParameter() {
   group=HealedScar_Sand
   parameterName=WispsNoise_Texture
   sortPriority=20
   texture=/Game/Global/Textures/Default/Tileable_Patterns/T_Blowing_Smoke_Pattern2_Tile_DISP
  }
  NamedRerouteDeclaration(None=TextureSample(UVs=BumpOffset(Coordinate=@n56, Height=@n57, HeightRatioInput=@n58) {heightRatio=0}, Tex=@n59) {desc=Front, samplerSource=SSM_Wrap_WorldGroupSettings}.R) {name=Front_Wisps_01, nodeColor=<rgba 0 0.9531 1 1>}
  n60 = VectorParameter() {
   group=HealedScar_Sand
   parameterName=WispsNoise2_UV
   channelNames {a=Offset Y, b=Offset X, g=Tilling Y, r=Tilling X}
   defaultValue=<rgba 1 1 0 0>
  }
  n61 = Multiply(
   A=@n51
   B=MaterialFunctionCall(X=@n60.Tilling X, Y=@n60.Tilling X) {materialFunction=&MakeFloat2}.Result
  )
  n62 = ComponentMask(None=MaterialFunctionCall(X=@n60.Offset X, Y=@n60.Offset X) {materialFunction=&MakeFloat2}.Result) {g=true, r=true}
  n63 = VectorParameter() {group=HealedScar_Sand, parameterName=WispsNoise2_OffsetSpeed, defaultValue=<rgba 0 0.01 0 0>}
  n64 = Add(
   A=Panner(
    Coordinate=Add(A=@n61, B=@n62)
    Speed=MaterialFunctionCall(X=@n63.R, Y=@n63.R) {materialFunction=&MakeFloat2}.Result
   )
   B=NamedRerouteUsage()
  )
  n65 = ScalarParameter() {defaultValue=-1.5, group=HealedScar_Sand, parameterName=Wisps2_Height}
  NamedRerouteDeclaration(None=TextureSample(UVs=BumpOffset(Coordinate=@n64, Height=@n65, HeightRatioInput=@n58) {heightRatio=0}, Tex=@n59) {desc=Front, samplerSource=SSM_Wrap_WorldGroupSettings}.R) {name=Front_Wisps_02, nodeColor=<rgba 0 0.9531 1 1>}
  n66 = VectorParameter() {
   group=HealedScar_Sand
   parameterName=WispsNoise3_UV
   channelNames {a=Offset Y, b=Offset X, g=Tilling Y, r=Tilling X}
   defaultValue=<rgba 1.25 1.25 0 0>
  }
  n67 = Multiply(
   A=@n51
   B=MaterialFunctionCall(X=@n66.Tilling X, Y=@n66.Tilling X) {materialFunction=&MakeFloat2}.Result
  )
  n68 = ComponentMask(None=MaterialFunctionCall(X=@n66.Offset X, Y=@n66.Offset X) {materialFunction=&MakeFloat2}.Result) {g=true, r=true}
  n69 = VectorParameter() {group=HealedScar_Sand, parameterName=WispsNoise3_OffsetSpeed, defaultValue=<rgba 0 0.05 0 0>}
  n70 = Add(
   A=Panner(
    Coordinate=Add(A=@n67, B=@n68)
    Speed=MaterialFunctionCall(X=@n69.R, Y=@n69.R) {materialFunction=&MakeFloat2}.Result
   )
   B=NamedRerouteUsage()
  )
  n71 = ScalarParameter() {defaultValue=-3, group=HealedScar_Sand, parameterName=Wisps3_Height}
  NamedRerouteDeclaration(None=TextureSample(UVs=BumpOffset(Coordinate=@n70, Height=@n71, HeightRatioInput=@n58) {heightRatio=0}, Tex=@n59) {desc=Front, samplerSource=SSM_Wrap_WorldGroupSettings}.R) {name=Front_Wisps_03, nodeColor=<rgba 0 0.9531 1 1>}
  n72 = NamedRerouteUsage()
  n73 = NamedRerouteUsage()
  n74 = NamedRerouteUsage()
  NamedRerouteDeclaration(None=MaterialFunctionCall!(XY vector=LinearInterpolate(A=@n72, B=@n73, Alpha=@n74), Z sign, Default Normalization vector) {materialFunction=/Engine/Functions/Engine_MaterialFunctions02/Math/DeriveNormalZ_Function}.Result) {name=Wisps_Normal, nodeColor=<rgba 0 1 0.8594 1>}
  n75 = VectorParameter() {
   group=HealedScar_Sand
   parameterName=Wisps_Secondary_DistortUV
   channelNames {a=Offset.Y, b=Offset.X, g=Tilling.Y, r=Tilling.X}
   defaultValue=<rgba 1 1 2 0>
  }
  n76 = Multiply(
   A=NamedRerouteUsage()
   B=MaterialFunctionCall(X=@n75.Tilling.X, Y=@n75.Tilling.X) {materialFunction=&MakeFloat2}.Result
  )
  n77 = MaterialFunctionCall(X=@n75.Offset.X, Y=@n75.Offset.X) {materialFunction=&MakeFloat2}
  n78 = VectorParameter() {group=HealedScar_Sand, parameterName=Wisps_Secondary_Flowmap_OffsetSpeed, defaultValue=<rgba 0.01 -0.04 0 0>}
  n79 = TextureSampleParameter2D(UVs=Panner(
    Coordinate=Add(A=@n76, B=@n77.Result)
    Speed=MaterialFunctionCall(X=@n78.R, Y=@n78.R) {materialFunction=&MakeFloat2}.Result
   )) {
   desc=Only RG channel works
   group=HealedScar_Sand
   parameterName=Wisps_DistortUV_Flowmap
   samplerSource=SSM_Wrap_WorldGroupSettings
   texture=&T_HealedScar_Sand_Flowmap
  }
  n80 = ComponentMask(None=@n79.RGB) {g=true, r=true}
  NamedRerouteDeclaration(None=Subtract(A=Multiply(A=@n80) {constB=2})) {name=FlowMap_Side_Secondary, nodeColor=<rgba 0 1 0.2969 1>}
  NamedRerouteDeclaration(None=VertexInterpolator(VS=PreSkinnedNormal()).PS) {name=PreskinnedNormal, nodeColor=<rgba 1 0.8672 0 1>}
  NamedRerouteDeclaration(None=ScalarParameter() {group=HealedScar_Sand, parameterName=GhostTransitionContrast, sortPriority=41}) {name=TransitionContrast, nodeColor=<rgba 0.2422 1 0 1>}
  n81 = NamedRerouteUsage()
  n82 = NamedRerouteUsage()
  n83 = NamedRerouteUsage()
  n84 = NamedRerouteUsage()
  n85 = LinearInterpolate(A=Add(A=Add(A=@n81, B=@n82), B=NamedRerouteUsage()), B=Add(A=Add(A=@n83, B=@n84), B=NamedRerouteUsage()), Alpha=NamedRerouteUsage()) {desc=Float3}
  n86 = NamedRerouteUsage()
  n87 = NamedRerouteUsage()
  n88 = Fresnel(
   ExponentIn=ScalarParameter() {defaultValue=0.5, group=HealedScar_Sand, parameterName=Wisps_Normal_Exponent}
   Normal=Transform(None=NamedRerouteUsage())
  ) {baseReflectFraction=0}
  n89 = ScalarParameter() {defaultValue=2, group=HealedScar_Sand, parameterName=Wisps_Fresnel_Strength}
  NamedRerouteDeclaration(None=Subtract(A=LinearInterpolate(A=@n85, B=@n86, Alpha=@n87) {desc=Float3}, B=Multiply(A=@n88, B=@n89))) {name=Wisps, nodeColor=<rgba 0 0.8702 1 1>}
  NamedRerouteDeclaration(None=NamedRerouteUsage()) {name=CachedFrontUV, nodeColor=<rgba 0.3359 0 1 1>}
  n90 = VectorParameter() {
   group=HealedScar_Sand
   parameterName=Wisps_Main_DistortUV
   channelNames {a=Offset.Y, b=Offset.X, g=Tilling.Y, r=Tilling.X}
   defaultValue=<rgba 0.5 0.2 0 0>
  }
  n91 = Add(
   A=Multiply(
    A=NamedRerouteUsage()
    B=MaterialFunctionCall(X=@n90.Tilling.X, Y=@n90.Tilling.X) {materialFunction=&MakeFloat2}.Result
   )
   B=MaterialFunctionCall(X=@n90.Offset.X, Y=@n90.Offset.X) {materialFunction=&MakeFloat2}.Result
  )
  n92 = VectorParameter() {group=HealedScar_Sand, parameterName=Wisps_Main_Flowmap_OffsetSpeed, defaultValue=<rgba -0.02 -0.02 0 0>}
  n93 = MaterialFunctionCall(X=@n92.R, Y=@n92.R) {materialFunction=&MakeFloat2}
  n94 = NamedRerouteUsage()
  n95 = ScalarParameter() {defaultValue=0.2, group=HealedScar_Sand, parameterName=FlowMap_Secondary_Scale}
  n96 = TextureSampleParameter2D(UVs=Add(A=Panner(Coordinate=@n91, Speed=@n93.Result), B=Multiply(A=@n94, B=@n95))) {
   desc=Only RG channel works
   group=HealedScar_Sand
   parameterName=Wisps_DistortUV_Flowmap
   samplerSource=SSM_Wrap_WorldGroupSettings
   texture=&T_HealedScar_Sand_Flowmap
  }
  n97 = Multiply(A=ComponentMask(None=@n96.RGB) {g=true, r=true}) {constB=2}
  NamedRerouteDeclaration(None=Multiply(A=Subtract(A=@n97), B=ScalarParameter() {defaultValue=0.5, group=HealedScar_Sand, parameterName=FlowMap_Main_Scale})) {name=FlowMap_Front_Main, nodeColor=<rgba 0 1 0.1797 1>}
  n98 = VectorParameter() {
   group=HealedScar_Sand
   parameterName=Wisps_Secondary_DistortUV
   channelNames {a=Offset.Y, b=Offset.X, g=Tilling.Y, r=Tilling.X}
   defaultValue=<rgba 1 0.5 2 0>
  }
  n99 = Multiply(
   A=NamedRerouteUsage()
   B=MaterialFunctionCall(X=@n98.Tilling.X, Y=@n98.Tilling.X) {materialFunction=&MakeFloat2}.Result
  )
  n100 = MaterialFunctionCall(X=@n98.Offset.X, Y=@n98.Offset.X) {materialFunction=&MakeFloat2}
  n101 = VectorParameter() {group=HealedScar_Sand, parameterName=Wisps_Secondary_Flowmap_OffsetSpeed, defaultValue=<rgba 0.01 -0.04 0 0>}
  n102 = TextureSampleParameter2D(UVs=Panner(
    Coordinate=Add(A=@n99, B=@n100.Result)
    Speed=MaterialFunctionCall(X=@n101.R, Y=@n101.R) {materialFunction=&MakeFloat2}.Result
   )) {
   desc=Only RG channel works
   group=HealedScar_Sand
   parameterName=Wisps_DistortUV_Flowmap
   samplerSource=SSM_Wrap_WorldGroupSettings
   texture=&T_HealedScar_Sand_Flowmap
  }
  n103 = ComponentMask(None=@n102.RGB) {g=true, r=true}
  NamedRerouteDeclaration(None=Subtract(A=Multiply(A=@n103) {constB=2})) {name=FlowMap_Front_Secondary, nodeColor=<rgba 0 1 0.2969 1>}
  NamedRerouteDeclaration(None=Multiply(A=@n17.A) {constB=0.05}) {name=BumpOffsetMask, nodeColor=<rgba 1 0 0.7031 1>}
  n104 = NamedRerouteUsage()
  n105 = VectorParameter() {
   group=HealedScar_Sand
   parameterName=WispsNoise1_UV
   channelNames {a=Offset Y, b=Offset X, g=Tilling Y, r=Tilling X}
   defaultValue=<rgba 0.5 0.5 0 0>
  }
  n106 = Multiply(
   A=@n104
   B=MaterialFunctionCall(X=@n105.Tilling X, Y=@n105.Tilling X) {materialFunction=&MakeFloat2}.Result
  )
  n107 = ComponentMask(None=MaterialFunctionCall(X=@n105.Offset X, Y=@n105.Offset X) {materialFunction=&MakeFloat2}.Result) {g=true, r=true}
  n108 = VectorParameter() {group=HealedScar_Sand, parameterName=WispsNoise1_OffsetSpeed, defaultValue=<rgba 0 0.02 0 0>}
  n109 = Add(
   A=Panner(
    Coordinate=Add(A=@n106, B=@n107)
    Speed=MaterialFunctionCall(X=@n108.R, Y=@n108.R) {materialFunction=&MakeFloat2}.Result
   )
   B=NamedRerouteUsage()
  )
  n110 = ScalarParameter() {group=HealedScar_Sand, parameterName=Wisps1_Height}
  n111 = NamedRerouteUsage()
  NamedRerouteDeclaration(None=TextureSample(UVs=BumpOffset(Coordinate=@n109, Height=@n110, HeightRatioInput=@n111) {heightRatio=0}, Tex=@n59) {desc=Front, samplerSource=SSM_Wrap_WorldGroupSettings}.R) {name=Side_Wisps_01, nodeColor=<rgba 0 0.9531 1 1>}
  n112 = VectorParameter() {
   group=HealedScar_Sand
   parameterName=WispsNoise2_UV
   channelNames {a=Offset Y, b=Offset X, g=Tilling Y, r=Tilling X}
   defaultValue=<rgba 1 1 -0.528 0.2>
  }
  n113 = Multiply(
   A=@n104
   B=MaterialFunctionCall(X=@n112.Tilling X, Y=@n112.Tilling X) {materialFunction=&MakeFloat2}.Result
  )
  n114 = ComponentMask(None=MaterialFunctionCall(X=@n112.Offset X, Y=@n112.Offset X) {materialFunction=&MakeFloat2}.Result) {g=true, r=true}
  n115 = VectorParameter() {group=HealedScar_Sand, parameterName=WispsNoise2_OffsetSpeed, defaultValue=<rgba 0 0.05 0 0>}
  n116 = Add(
   A=Panner(
    Coordinate=Add(A=@n113, B=@n114)
    Speed=MaterialFunctionCall(X=@n115.R, Y=@n115.R) {materialFunction=&MakeFloat2}.Result
   )
   B=NamedRerouteUsage()
  )
  n117 = ScalarParameter() {defaultValue=-1.5, group=HealedScar_Sand, parameterName=Wisps2_Height}
  NamedRerouteDeclaration(None=TextureSample(UVs=BumpOffset(Coordinate=@n116, Height=@n117, HeightRatioInput=@n111) {heightRatio=0}, Tex=@n59) {desc=Front, samplerSource=SSM_Wrap_WorldGroupSettings}.R) {name=Side_Wisps_02, nodeColor=<rgba 0 0.9531 1 1>}
  n118 = VectorParameter() {
   group=HealedScar_Sand
   parameterName=WispsNoise3_UV
   channelNames {a=Offset Y, b=Offset X, g=Tilling Y, r=Tilling X}
   defaultValue=<rgba 2 2 0.2 0.4>
  }
  n119 = Multiply(
   A=@n104
   B=MaterialFunctionCall(X=@n118.Tilling X, Y=@n118.Tilling X) {materialFunction=&MakeFloat2}.Result
  )
  n120 = ComponentMask(None=MaterialFunctionCall(X=@n118.Offset X, Y=@n118.Offset X) {materialFunction=&MakeFloat2}.Result) {g=true, r=true}
  n121 = VectorParameter() {group=HealedScar_Sand, parameterName=WispsNoise3_OffsetSpeed, defaultValue=<rgba 0 0.05 0 0>}
  n122 = Add(
   A=Panner(
    Coordinate=Add(A=@n119, B=@n120)
    Speed=MaterialFunctionCall(X=@n121.R, Y=@n121.R) {materialFunction=&MakeFloat2}.Result
   )
   B=NamedRerouteUsage()
  )
  n123 = ScalarParameter() {defaultValue=-3, group=HealedScar_Sand, parameterName=Wisps3_Height}
  NamedRerouteDeclaration(None=TextureSample(UVs=BumpOffset(Coordinate=@n122, Height=@n123, HeightRatioInput=@n111) {heightRatio=0}, Tex=@n59) {desc=Front, samplerSource=SSM_Wrap_WorldGroupSettings}.R) {name=Side_Wisps_03, nodeColor=<rgba 0 0.9531 1 1>}
  n124 = NamedRerouteUsage()
  n125 = Saturate(None=ComponentMask(None=@n124) {g=true})
  n126 = Saturate(None=ComponentMask(None=Multiply(A=@n124) {constB=-1}) {g=true})
  NamedRerouteDeclaration(None=SmoothStep(Value=Add(A=@n125, B=@n126)) {constMin=0.4}) {name=SideFrontBlend, nodeColor=<rgba 1 0.4219 0 1>}
  n127 = Saturate(None=ComponentMask(None=@n124) {b=true})
  n128 = Saturate(None=ComponentMask(None=Multiply(A=@n124) {constB=-1}) {b=true})
  NamedRerouteDeclaration(None=SmoothStep(Value=Add(A=@n127, B=@n128)) {constMin=0.4}) {name=TopBlend, nodeColor=<rgba 1 0.4219 0 1>}
  NamedRerouteDeclaration(None=TextureSample(UVs=NamedRerouteUsage(), Tex=@n59) {desc=Top, samplerSource=SSM_Wrap_WorldGroupSettings}.R) {name=Top_Wisps, nodeColor=<rgba 0 0.9531 1 1>}
  n129 = SmoothStep(Value=ComponentMask(None=@n4) {b=true}) {constMax=380, constMin=75, desc=body}
  n130 = ComponentMask(None=@n4) {b=true}
  n131 = ComponentMask(None=@n4) {b=true}
  n132 = ComponentMask(None=@n4) {g=true}
  n133 = Abs(None=ComponentMask(None=@n4) {r=true})
  n134 = Add(
   A=Add(A=SmoothStep(Value=@n130) {constMax=-15, constMin=10, desc=feet}, B=SmoothStep(Value=@n131) {constMax=180, constMin=160, desc=head})
   B=Add(A=SmoothStep(Value=@n132) {constMax=-60, constMin=-15, desc=tail}, B=SmoothStep(Value=@n133) {constMax=100, constMin=50, desc=Hand})
  )
  n135 = OneMinus(None=NamedRerouteUsage())
  NamedRerouteDeclaration(None=Add(A=Multiply(A=@n129) {constB=0.5}, B=Multiply(A=@n134, B=@n135))) {name=MaskedEmissive, nodeColor=<rgba 0.1953 0 1 1>}
  n136 = Time()
  n137 = VectorParameter() {group=HealedScar_Sand, parameterName=Wisps_Speed, defaultValue=<rgba 0 0 1 0>}
  n138 = ScalarParameter() {defaultValue=64, group=HealedScar_Sand, parameterName=WispsTextureSize}
  n139 = Divide(A=Add(A=NamedRerouteUsage(), B=Multiply(A=@n136, B=@n137.RGB)), B=Multiply(A=Abs(None=@n138)) {constB=-1})
  NamedRerouteDeclaration(None=Multiply(A=ComponentMask(None=@n139) {b=true, r=true}, B=Constant2Vector() {g=1, r=-1})) {name=Triplanar_Front_UV, nodeColor=<rgba 0 0.1328 1 1>}
  NamedRerouteDeclaration(None=ComponentMask(None=@n139) {b=true, g=true}) {name=Triplanar_Side_UV, nodeColor=<rgba 0 0.1328 1 1>}
  NamedRerouteDeclaration(None=ComponentMask(None=@n139) {g=true, r=true}) {name=Triplanar_Top_UV, nodeColor=<rgba 0 0.1328 1 1>}
  NamedRerouteDeclaration(None=NamedRerouteUsage()) {name=CachedSideUV, nodeColor=<rgba 0.3359 0 1 1>}
  NamedRerouteDeclaration(None=Add(A=@n17.R, B=@n17.R)) {name=ArmorMask, nodeColor=<rgba 0.4297 0 1 1>}
 ]}
