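# Parent material for the Sand skin of the HealedScar character body that draws a soft, translucent, unlit glow outline on a skeletal mesh.
# A custom HLSL node takes noise- and TAA-jittered radial scene-depth samples, with radius scaled by camera distance and screen size, to build a blurred edge mask.
# That mask, optional inner glow and Fresnel shape opacity, while GlowColor drives emissive.
# Vertex normals inflate the shell through world position offset.
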Material /BRCosmetics/Characters/Player/Male/Large/Bodies/M_LRG_HealedScar/Skins/Sand/Materials/M_HealedScar_Sand_SoftGlowOutlines_PARENT
graph {
 properties {
  bAutomaticallySetUsageInEditor=false
  bUseMaterialAttributes=true
  bUsedWithMeshDeformer=true
  bUsedWithSkeletalMesh=true
  blendMode=BLEND_Translucent
  pixelDepthOffsetMode=PDOM_Legacy
  previewMesh=/BRCosmetics/Characters/Player/Male/Large/Bodies/M_LRG_HealedScar/Meshes/M_LRG_HealedScar
  shadingModel=MSM_Unlit
  twoSided=true
  usedShadingModels=Unlit
 }
 nodes [
  n1 = LinearInterpolate(
   A=ScalarParameter() {defaultValue=3, group=SoftGlow, parameterName=OutlineThickness_Far}
   B=ScalarParameter() {defaultValue=2.7, group=SoftGlow, parameterName=OutlineThickness}
   Alpha=NamedRerouteUsage()
  )
  n2 = Multiply(A=SceneTexelSize()) {constB=1440}
  n3 = MaterialFunctionCall(Vector 3=Subtract(A=ParticlePositionWS(), B=CameraPositionWS()), Vector 2=Constant2Vector()) {materialFunction=/Engine/Functions/Engine_MaterialFunctions02/Utility/VectorLength}
  NamedRerouteDeclaration(None=Divide(A=Divide(A=@n1, B=@n2), B=Clamp(None=@n3.V3 Length) {clampMode=CMODE_ClampMin, minDefault=30})) {name=Distance, nodeColor=<rgba 0.8984 1 0 1>}
  NamedRerouteDeclaration(None=ScalarParameter() {defaultValue=3, group=SoftGlow, parameterName=DistanceSteps}) {name=DistanceSteps, nodeColor=<rgba 0.7812 0 1 1>}
  n4 = Constant() {r=0.6}
  MaterialExpressionNamedRerouteUsage_1 = NamedRerouteUsage()
  MaterialExpressionNamedRerouteUsage_5 = NamedRerouteUsage()
  MaterialExpressionNamedRerouteUsage_7 = NamedRerouteUsage()
  MaterialExpressionNamedRerouteUsage_42 = NamedRerouteUsage()
  MaterialExpressionNamedRerouteUsage_43 = NamedRerouteUsage()
  MaterialExpressionNamedRerouteUsage_10 = NamedRerouteUsage()
  MaterialExpressionNamedRerouteUsage_44 = NamedRerouteUsage()
  MaterialExpressionNamedRerouteUsage_45 = NamedRerouteUsage()
  MaterialExpressionNamedRerouteUsage_16 = NamedRerouteUsage()
  MaterialExpressionNamedRerouteUsage_22 = NamedRerouteUsage()
  MaterialExpressionNamedRerouteUsage_20 = NamedRerouteUsage()
  n5 = Custom!(
   Distance=@MaterialExpressionNamedRerouteUsage_1
   DistanceSteps=@MaterialExpressionNamedRerouteUsage_5
   RadialSteps=@MaterialExpressionNamedRerouteUsage_7
   TempAARotation=@MaterialExpressionNamedRerouteUsage_42
   DistanceMask=@MaterialExpressionNamedRerouteUsage_43
   Tex=@MaterialExpressionNamedRerouteUsage_10
   TempAADistance=@MaterialExpressionNamedRerouteUsage_44
   RadialOffset=@MaterialExpressionNamedRerouteUsage_45
   DivideF=@MaterialExpressionNamedRerouteUsage_16
   PDepth=@MaterialExpressionNamedRerouteUsage_22
   PixelDepthOffset=@MaterialExpressionNamedRerouteUsage_20
  ) {
   code="float CurColor = 0;\r\nfloat2 BaseUV = MaterialFloat2(ScreenAlignedPosition(Parameters.ScreenPosition).xy);\r\nfloat2 NewUV = BaseUV;\r\nfloat StepSize = Distance / (int) DistanceSteps;\r\nfloat CurDistance = 0;\r\nfloat2 CurOffset = 0;\r\nfloat TwoPi = 6.283185;\r\nfloat Substep = 0;\r\nfloat2 ScenePixels=View.BufferSizeAndInvSize.xy*BaseUV;\r\nScenePixels+=View.TemporalAAParams.r;\r\nfloat2 RandomSamp = ((uint)(ScenePixels.x) + 2 * (uint)(ScenePixels.y)) % 5;\r\nRandomSamp+=Texture2DSample(Tex,TexSampler,ScenePixels);\r\nRandomSamp/=5;\r\nRandomSamp-=0.5;\r\nTempAARotation*=RandomSamp;\r\nTempAADistance*=StepSize*RandomSamp;\r\n\r\n\r\nint i=0;\r\nif (DistanceSteps<1)\r\n{\r\n    Mask = 0;\r\n    return 0;\r\n}\r\nelse\r\n{\r\n    //CurDistance += 0.5*StepSize;\r\n    while ( i < (int) DistanceSteps)\r\n    {\r\n        //CurDistance+=StepSize;\r\n        for (int j = 0; j < (int) RadialSteps; j++)\r\n        {\r\n            CurOffset.x = cos(TwoPi*((TempAARotation+Substep) / RadialSteps));\r\n            CurOffset.y = sin(TwoPi*((TempAARotation+Substep) / RadialSteps)) * View.BufferSizeAndInvSize.x / View.BufferSizeAndInvSize.y;\r\n            CurOffset *=DistanceMask;\r\n            NewUV.x = BaseUV.x + (CurOffset.x * (CurDistance+(RandomSamp*TempAADistance)));\r\n            NewUV.y = BaseUV.y + (CurOffset.y * (CurDistance+(RandomSamp*TempAADistance)));\r\n            float tmpdepth = CalcSceneDepth(NewUV)-PDepth+PixelDepthOffset;\r\n            CurColor += (min(1-saturate(max(0,(tmpdepth))/(DivideF+PixelDepthOffset))-0.99,0)+0.99)*saturate(sign(tmpdepth));\r\n            //CurDistance+=(StepSize+(TempAADistance))/RadialSteps;\r\n            Substep++;\r\n        }\r\n        CurDistance+=StepSize;\r\n        Substep+=RadialOffset;\r\n        i++;\r\n    }\r\n    float tmpdepth2 = CalcSceneDepth(BaseUV)-PDepth+PixelDepthOffset;\r\n    float mask = (min(1-saturate(max(0,(tmpdepth2))/(DivideF+PixelDepthOffset))-0.99,0)+0.99)*saturate(sign(tmpdepth2));\r\n    Mask = mask;\r\n    CurColor = CurColor / ((int)DistanceSteps*(int)RadialSteps);\r\n    return CurColor;\r\n}"
   desc="\n*** Original source before expression upgrade ***\nfloat3 CurColor;\r\nfloat2 BaseUV = MaterialFloat2(ScreenAlignedPosition(Parameters.ScreenPosition).xy);\r\nfloat2 NewUV = BaseUV;\r\nfloat StepSize = Distance / (int) DistanceSteps;\r\nfloat CurDistance=0;\r\nfloat2 CurOffset;\r\nfloat TwoPi = 6.283185;\r\nfloat Substep;\r\nfloat2 ScenePixels=View.RenderTargetSize*BaseUV;\r\nScenePixels+=View.TemporalAAParams.r;\r\nfloat2 RandomSamp = ((uint)(ScenePixels.x) + 2 * (uint)(ScenePixels.y)) % 5;\r\nRandomSamp+=Texture2DSample(Tex,TexSampler,ScenePixels);\r\nRandomSamp/=5;\r\nRandomSamp-=0.5;\r\nTempAARotation*=RandomSamp;\r\nTempAADistance*=StepSize*RandomSamp;\r\n\r\n\r\nint i=0;\r\nif (DistanceSteps<1)\r\n{\r\nreturn DecodeSceneColorForMaterialNode(NewUV);\r\n}\r\nelse\r\n{\r\n//CurDistance += 0.5*StepSize;\r\nwhile ( i < (int) DistanceSteps)\r\n{\r\n\r\n//CurDistance+=StepSize;\r\nfor (int j = 0; j < (int) RadialSteps; j++)\r\n{\r\nCurOffset.x = cos(TwoPi*((TempAARotation+Substep) / RadialSteps));\r\nCurOffset.y = sin(TwoPi*((TempAARotation+Substep) / RadialSteps));\r\nCurOffset *=DistanceMask;\r\nNewUV.x = BaseUV.x + (CurOffset.x * (CurDistance+(RandomSamp*TempAADistance)));\r\nNewUV.y = BaseUV.y + (CurOffset.y * (CurDistance+(RandomSamp*TempAADistance)));\r\nCurColor += DecodeSceneColorForMaterialNode(NewUV);\r\n//CurDistance+=(StepSize+(TempAADistance))/RadialSteps;\r\nSubstep++;\r\n}\r\nCurDistance+=StepSize;\r\nSubstep+=RadialOffset;\r\ni++;\r\n}\r\nCurColor = CurColor / ((int)DistanceSteps*(int)RadialSteps);\r\nreturn CurColor;\r\n}\n*** Original source before expression upgrade ***\nfloat3 CurColor;\r\nfloat2 BaseUV = MaterialFloat2(ScreenAlignedPosition(Parameters.ScreenPosition).xy);\r\nfloat2 NewUV = BaseUV;\r\nfloat StepSize = Distance / (int) DistanceSteps;\r\nfloat CurDistance=0;\r\nfloat2 CurOffset;\r\nfloat TwoPi = 6.283185;\r\nfloat Substep;\r\nfloat2 ScenePixels=Frame.RenderTargetSize*BaseUV;\r\nScenePixels+=Frame.TemporalAAParams.r;\r\nfloat2 RandomSamp = ((uint)(ScenePixels.x) + 2 * (uint)(ScenePixels.y)) % 5;\r\nRandomSamp+=Texture2DSample(Tex,TexSampler,ScenePixels);\r\nRandomSamp/=5;\r\nRandomSamp-=0.5;\r\nTempAARotation*=RandomSamp;\r\nTempAADistance*=StepSize*RandomSamp;\r\n\r\n\r\nint i=0;\r\nif (DistanceSteps<1)\r\n{\r\nreturn DecodeSceneColorForMaterialNode(NewUV);\r\n}\r\nelse\r\n{\r\n//CurDistance += 0.5*StepSize;\r\nwhile ( i < (int) DistanceSteps)\r\n{\r\n\r\n//CurDistance+=StepSize;\r\nfor (int j = 0; j < (int) RadialSteps; j++)\r\n{\r\nCurOffset.x = cos(TwoPi*((TempAARotation+Substep) / RadialSteps));\r\nCurOffset.y = sin(TwoPi*((TempAARotation+Substep) / RadialSteps));\r\nCurOffset *=DistanceMask;\r\nNewUV.x = BaseUV.x + (CurOffset.x * (CurDistance+(RandomSamp*TempAADistance)));\r\nNewUV.y = BaseUV.y + (CurOffset.y * (CurDistance+(RandomSamp*TempAADistance)));\r\nCurColor += DecodeSceneColorForMaterialNode(NewUV);\r\n//CurDistance+=(StepSize+(TempAADistance))/RadialSteps;\r\nSubstep++;\r\n}\r\nCurDistance+=StepSize;\r\nSubstep+=RadialOffset;\r\ni++;\r\n}\r\nCurColor = CurColor / ((int)DistanceSteps*(int)RadialSteps);\r\nreturn CurColor;\r\n}"
   outputType=CMOT_Float1
   showCode=true
   additionalOutputs [{outputName=Mask, outputType=CMOT_Float1}]
   inputs [
    {
     inputName=Distance
     input {expression=&self:MaterialExpressionNamedRerouteUsage_1}
    }
    {
     inputName=DistanceSteps
     input {expression=&self:MaterialExpressionNamedRerouteUsage_5}
    }
    {
     inputName=RadialSteps
     input {expression=&self:MaterialExpressionNamedRerouteUsage_7}
    }
    {
     inputName=TempAARotation
     input {expression=&self:MaterialExpressionNamedRerouteUsage_42}
    }
    {
     inputName=DistanceMask
     input {expression=&self:MaterialExpressionNamedRerouteUsage_43}
    }
    {
     inputName=Tex
     input {expression=&self:MaterialExpressionNamedRerouteUsage_10}
    }
    {
     inputName=TempAADistance
     input {expression=&self:MaterialExpressionNamedRerouteUsage_44}
    }
    {
     inputName=RadialOffset
     input {expression=&self:MaterialExpressionNamedRerouteUsage_45}
    }
    {
     inputName=DivideF
     input {expression=&self:MaterialExpressionNamedRerouteUsage_16}
    }
    {
     inputName=PDepth
     input {expression=&self:MaterialExpressionNamedRerouteUsage_22}
    }
    {
     inputName=PixelDepthOffset
     input {expression=&self:MaterialExpressionNamedRerouteUsage_20}
    }
   ]
  }
  n6 = Saturate(None=Subtract(A=@n5.return, B=Multiply(A=@n5.Mask, B=@n4)))
  n7 = ScalarParameter() {defaultValue=1, group=SoftGlow, parameterName=OutlineGlowStrength}
  n8 = Multiply(A=@n6, B=@n7)
  n9 = Multiply(
   A=ScalarParameter() {defaultValue=0.5, group=SoftGlow, parameterName=InnerGlowStrength}
   B=Power(Base=OneMinus(None=@n6), Exp=ScalarParameter() {defaultValue=10, group=SoftGlow, parameterName=InnerGlowPower})
  )
  n10 = Constant() {r=2.5}
  n11 = Saturate(None=Multiply(A=Power(Base=@n5.Mask, Exp=@n10), B=ScalarParameter() {defaultValue=100, group=SoftGlow, parameterName=InnerGlowMaxArea}) {constB=1000})
  n12 = Subtract(A=Saturate(None=@n5.return), B=Saturate(None=@n5.Mask))
  n13 = Saturate(None=StaticSwitch(True=LinearInterpolate(A=@n8, B=@n9, Alpha=@n11) {constB=0}, False=Multiply(A=@n12, B=@n7), Value=NamedRerouteUsage()))
  n14 = Fresnel(ExponentIn=ScalarParameter() {defaultValue=1, group=SoftGlow, parameterName=Fresnel_Exponent}) {baseReflectFraction=0, exponent=1}
  n15 = Saturate(None=Power(Base=OneMinus(None=@n14), Exp=ScalarParameter() {group=SoftGlow, parameterName=Fresnel_Power}))
  NamedRerouteDeclaration(None=Multiply(A=Multiply(A=@n13, B=@n15), B=ScalarParameter() {defaultValue=1, group=SoftGlow, parameterName=OpacityMulti})) {name=Opacity_OUT, nodeColor=<rgba 1 0.5625 0 1>}
  NamedRerouteDeclaration(None=MaterialFunctionCall(Emissive=VectorParameter() {group=SoftGlow, parameterName=GlowColor, defaultValue=<rgba 4.8649 0.747 15 0.08>}.RGB) {materialFunction=&MF_FX_EmissiveAdjustments}.Result) {name=Emissive_OUT, nodeColor=<rgba 0.2422 0 1 1>}
  NamedRerouteDeclaration(None=Constant() {r=1}) {name=TempAARotation, nodeColor=<rgba 1 0 0.8203 1>}
  NamedRerouteDeclaration(None=ScalarParameter() {defaultValue=1, group=SoftGlow, parameterName=DistanceMask_Multi}) {name=DistanceMask, nodeColor=<rgba 0 1 0.8594 1>}
  NamedRerouteDeclaration(None=ScalarParameter() {defaultValue=1, group=SoftGlow, parameterName=TempAADistance}) {name=TempAADistance, nodeColor=<rgba 1 0.5391 0 1>}
  NamedRerouteDeclaration(None=ScalarParameter() {defaultValue=0.62, group=SoftGlow, parameterName=RadialOffset}) {name=RadialOffset, nodeColor=<rgba 0.2188 0 1 1>}
  NamedRerouteDeclaration(None=ScalarParameter() {defaultValue=2, group=SoftGlow, parameterName=RadialSteps}) {name=RadialSteps, nodeColor=<rgba 0 1 0.4609 1>}
  NamedRerouteDeclaration(None=@n5.Mask) {name=BlurAreaDepth0-1, nodeColor=<rgba 1 0 0.1172 1>}
  n16 = Length(None=ActorPositionWS() {originType=CameraRelative}) {is=class MaterialExpressionLength}
  n17 = ScalarParameter() {defaultValue=100, group=SoftGlow, parameterName=DistanceBlending_Bias}
  n18 = Power(Base=Divide(A=Subtract(A=@n16, B=@n17) {constB=100}) {constB=400}) {constExponent=0.5}
  NamedRerouteDeclaration(None=OneMinus(None=Saturate(None=@n18))) {name=DistanceBlending, nodeColor=<rgba 0.4766 1 0 1>}
  NamedRerouteDeclaration(None=StaticBoolParameter() {group=SoftGlow, parameterName=UseInnerGlow, sortPriority=0}) {name=bUseInnerGlow, nodeColor=<rgba 0.0781 1 0 1>}
  NamedRerouteDeclaration(None=TextureObject() {samplerType=SAMPLERTYPE_LinearGrayscale, texture=/Engine/EngineMaterials/Good64x64TilingNoiseHighFreq}) {name=Tex, nodeColor=<rgba 1 0 0.8203 1>}
  n19 = NamedRerouteDeclaration(None=SceneDepth().R) {name=SceneDepth, nodeColor=<rgba 0.4297 1 0 1>}
  NamedRerouteDeclaration(None=Multiply(A=Clamp(None=@n19) {minDefault=1}, B=PixelDepth().R)) {name=PDepth, nodeColor=<rgba 0 1 0.8594 1>}
  NamedRerouteDeclaration(None=VertexInterpolator(VS=LocalPosition() {localOrigin=InstancePreSkinning}.XYZ).PS) {name=PreSkinnedLocation, nodeColor=<rgba 0.5469 1 0 1>}
  NamedRerouteDeclaration(None=ScalarParameter() {group=SoftGlow, parameterName=PixelDepthOffset}) {name=PixelDepthOffset, nodeColor=<rgba 0.2188 0 1 1>}
  NamedRerouteDeclaration(None=ScalarParameter() {defaultValue=10, group=SoftGlow, parameterName=MaxBlurDepth}) {name=MaxBlurDepth, nodeColor=<rgba 0.1016 1 0 1>}
  n20 = NamedRerouteUsage()
  NamedRerouteDeclaration(None=SmoothStep(Value=ComponentMask(None=@n20) {b=true}) {constMax=50, constMin=10}) {name=BodyMask, nodeColor=<rgba 1 0 0.8672 1>}
  NamedRerouteDeclaration(None=Multiply(A=VertexNormalWS(), B=ScalarParameter() {defaultValue=5, group=SoftGlow, parameterName=BagThickness})) {name=WPO_OUT, nodeColor=<rgba 0 1 0.8828 1>}
  SetMaterialAttributes!(
   MaterialAttributes
   Emissive Color=LinearInterpolate(B=NamedRerouteUsage(), Alpha=NamedRerouteUsage())
   Opacity=LinearInterpolate(B=NamedRerouteUsage(), Alpha=NamedRerouteUsage())
   World Position Offset=NamedRerouteUsage()
  ) {attributeSetTypes [#1, #2, #3]}
 ]
}
